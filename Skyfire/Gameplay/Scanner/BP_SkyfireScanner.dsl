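; Server-authoritative Skyfire scanner that hops between tagged warp points and periodically switches on a light cone and trigger box.
; On match start it listens for game phase changes; the Aircraft phase caches warp locations, stops listening and starts the delayed scan loop.
; Each cycle teleports to a random location excluding the current one, waits to telegraph, enables query collision for the scan, then repeats until bSkyfireScanPortionOver.
; No overlap handling appears here.
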(event EventEndPlay
  (params (EndPlayReason :type EEndPlayReason%20Enum :default "Destroyed"))
  (bind _gamephaseupdatedasynctask (Variables|Default|GetGamePhaseUpdatedAsyncTask)
    (:then
      (GameplayEventRouter|StopListeningtoStatefulEvent _gamephaseupdatedasynctask))
    (:else)))

(event EventOnMatchStarted
  ;[0 1 0.2334] Event on Match Started - set inital states for components and bind to OnGamephaseChanged
  (Networking|SwitchHasAuthority
    (:Authority
      (Collision|SetCollisionEnabled (Variables|Default|GetScannerTriggerBox) "NoCollision")
      (Rendering|SetVisibility (Variables|Default|GetS_Intro_Photos_SimpleLightCone_Scanner) false true)
      (Rendering|SetVisibility (Variables|Default|GetSaucer) true true)
      (bind _returnvalue (GameplayEventRouter|GetEventRouter "/Script/FortniteGame.FortGameState" self))
      (bind (_asynctaskproxy _payload) (GameplayEventRouter|StartListeningToStatefulEvent _returnvalue "/Script/FortniteGame.FortGamePhaseUpdatedEvent")
        (:then
          (Variables|Default|SetGamePhaseUpdatedAsyncTask _asynctaskproxy))
        (:OnEventReceived
          _)
        (:OnSavedEventStateExists
          _))
      (CallFunction|OnGamePhaseChanged (Utilities|Struct|BreakFortGamePhaseUpdatedEvent _payload)))))

(event Custom|OnSkyrfireStart
  ;[0.4526 1 0.8066] OnSkyfireStart - wait some time before starting to scan
  (Networking|SwitchHasAuthority
    (:Authority
      (Utilities|FlowControl|Delay (Variables|Default|GetTEMP_WaitTimeBeforeStartingToScan))
      (CallFunction|TeleporttoNewLocationandScan))))

(event Custom|Teleport%20To%20New%20Location%20And%20Scan
  (bind _newlocation (CallFunction|GetNextAvailableScannerLocation))
  (bind _scannertriggerbox (Variables|Default|GetScannerTriggerBox))
  ;[1 0.442 0.09453] TeleportToNewLocationAndScan
  (Networking|SwitchHasAuthority
    (:Authority
      (Utilities|FlowControl|Sequence
        (:then_0
          ;;[0.6562 1 0] Set Visibility to saucer and scanning light
          (Rendering|SetVisibility (Variables|Default|GetSaucer) true false)
          (Rendering|SetVisibility (Variables|Default|GetS_Intro_Photos_SimpleLightCone_Scanner) false true))
        (:then_1
          ;;[0.1352 0.9801 1] Get the next lcoation to teleport to and teleport there is valid
          (bind _returnvalue (Transformation|Teleport _newlocation (Transformation|GetActorRotation)))
          (if _returnvalue
            (Variables|Default|SetCurrentLocation _newlocation)
            (else
              (Debugging|LogString "NewEnumerator0" "NewEnumerator0" "Scanner Failed to teleport to new location" true))))
        (:then_2
          ;;[1 0.237 0.8868] Wait at new location before turing on the scanning light and beginning to scan
          (Utilities|FlowControl|Delay (Variables|Default|GetTEMPTelegraphScanTime))
          (Collision|SetCollisionEnabled _scannertriggerbox "QueryOnly")
          (Rendering|SetVisibility (Variables|Default|GetS_Intro_Photos_SimpleLightCone_Scanner) true true)
          (Utilities|FlowControl|Delay (Variables|Default|GetTEMPScanTime))
          (Collision|SetCollisionEnabled _scannertriggerbox "NoCollision")
          (if (Variables|Default|GetbIsFirstScan)
            (Variables|Default|SetbIsFirstScan false)
            (else
              _))
          (if (not (Variables|Default|GetbSkyfireScanPortionOver))
            (CallFunction|TeleporttoNewLocationandScan)))))))

(event Custom|OnGamePhaseChanged
  (params (GamePhase :type EAthenaGamePhase%20Enum))
  ;[1 0.6162 0.08689] OnGamephaseChanged, SetTeleport actors and begin and call OnSkyfireStart, unbind from event
  (Networking|SwitchHasAuthority
    (:Authority
      (switch Utilities|FlowControl|Switch|SwitchonEAthenaGamePhase GamePhase
        (:Aircraft
          (CallFunction|SetTeleportActors)
          (CallFunction|OnSkyrfireStart)
          (bind _gamephaseupdatedasynctask (Variables|Default|GetGamePhaseUpdatedAsyncTask)
            (:then
              (GameplayEventRouter|StopListeningtoStatefulEvent _gamephaseupdatedasynctask))
            (:else)))))))

(fn GetNextAvailableScannerLocation
  ;[1 0.6843 0.1036] Get the next available scanner location, removing the current location from the list so we dont get the same spot twice.
  (bind _output_get (LocalVariables|SetAvailableScannerLocations (Variables|Default|GetAllScannerLocations)))
  (if (not (Variables|Default|GetbIsFirstScan))
    (Utilities|Array|RemoveItem _output_get (Variables|Default|GetCurrentLocation)))
  (if (Utilities|Array|IsValidIndex _output_get (Math|Random|RandomIntegerinRange 0 (- (Utilities|Array|Length _output_get) 1)))
    (return (Utilities|Array|Get(acopy) _output_get (Math|Random|RandomIntegerinRange 0 (- (Utilities|Array|Length _output_get) 1))))
    (else
      (Debugging|LogString "NewEnumerator0" "NewEnumerator0" "GetNewLocation Failed to get valid Index for AvailableScannerLocations" true
        (:Then
          (return "0.000000, 0, 0"))))))

(fn SetTeleportActors
  ;[0 1 0.3007] Set all the scanning points found
  (bind _outactors (Actor|GetAllActorswithTag (Variables|Default|GetScannerWarpTag)))
  (for _array_element _outactors
    (Utilities|Array|Add (Variables|Default|GetAllScannerLocations) (Transformation|GetActorLocation _array_element))))

(fn ConstructionScript)
