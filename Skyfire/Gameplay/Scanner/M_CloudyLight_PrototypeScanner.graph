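# Translucent, unlit, two-sided material for a prototype scanner, producing a soft, cloudy, warm-tinted light effect.
# Two samples of one noise cutout texture, stretched along V and panning along U at different speeds, are summed and tinted pale gold.
# Pixel depth dims the result near the camera, and DepthFade softens intersections with geometry.
# An inverted V gradient fades it along the mesh before a final scale-down.
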Material /Skyfire/Gameplay/Scanner/M_CloudyLight_PrototypeScanner
graph {
 properties {blendMode=BLEND_Translucent, pixelDepthOffsetMode=PDOM_Legacy, shadingModel=MSM_Unlit, twoSided=true, usedShadingModels=Unlit}
 nodes [
  n1 = MaterialFunctionCall(UV Channel=TextureCoordinate()) {materialFunction=&LinearGradient}
  n2 = Panner(Coordinate=TextureCoordinate() {vTiling=0.005}) {speedX=0.01}
  n3 = Panner(Coordinate=TextureCoordinate() {vTiling=0.075}) {speedX=0.006}
  n4 = Multiply(
   A=Constant3Vector() {constant=<rgba 0.835 0.7353 0.3424 1>}
   B=Add(
    A=TextureSample(UVs=@n2) {texture=/Game/Effects/Fort_Effects/Textures/Misc/T_Shot0030_NoiseCutout}.G
    B=TextureSample(UVs=@n3) {texture=&T_Shot0030_NoiseCutout}.R
   )
  )
  n5 = Constant() {r=2}
  n6 = Divide(A=PixelDepth().R, B=Constant() {r=512})
  n7 = Multiply(A=Multiply(A=Multiply(A=@n4, B=@n5)), B=Multiply(B=Clamp(None=@n6)) {constA=1})
  n8 = Constant() {r=128}
  Divide(A=Multiply(A=OneMinus(None=@n1.VGradient), B=DepthFade(Opacity=@n7, FadeDistance=@n8)), B=Constant() {r=10})
 ]
}
